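# Physics asset for the male medium Foe head cosmetic, giving physics bodies to a left-ear bone chain.
# Four identical capsules cover the ear segments, with a tiny capsule on the master bone at the root.
# The first segment swings and twists freely against the master, while each later segment constrains to the previous one with narrow swing limits and zero twist.
# Segment shapes carry hat_dyn names despite sitting on ear bones.
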PhysicsAsset /BRCosmetics/Characters/Player/Male/Medium/Heads/M_MED_Foe_Head/Meshes/Male_Medium_Foe_Head_Physics
bodies {
 dyn_ear01_l [{shapeType=Capsule, length=5, radius=5, shapeName=hat_dyn1_capsule, center=<5 0 0>, rotation=<rot -90 0 0>}]
 dyn_ear02_l [{shapeType=Capsule, length=5, radius=5, shapeName=hat_dyn2_capsule, center=<5 0 0>, rotation=<rot -90 0 0>}]
 dyn_ear03_l [{shapeType=Capsule, length=5, radius=5, shapeName=hat_dyn3_capsule, center=<5 0 0>, rotation=<rot -90 0 0>}]
 dyn_ear04_l [{shapeType=Capsule, length=5, radius=5, shapeName=hat_dyn4_capsule, center=<5 0 0>, rotation=<rot -90 0 0>}]
 dyn_ear_l_master [{shapeType=Capsule, length=0.505, radius=0.505, shapeName=dyn_ear_l_master_capsule, rotation=<rot 0 -90 0>}]
}
constraints [
 {
  bone1Name=dyn_ear01_l
  bone2Name=dyn_ear_l_master
  swing1LimitDegrees=45
  swing1Motion=Free
  swing2LimitDegrees=45
  swing2Motion=Free
  twistLimitDegrees=45
  twistMotion=Free
 }
 {bone1Name=dyn_ear02_l, bone2Name=dyn_ear01_l, swing1LimitDegrees=7, swing2LimitDegrees=10, twistLimitDegrees=0}
 {bone1Name=dyn_ear03_l, bone2Name=dyn_ear02_l, swing1LimitDegrees=7, swing2LimitDegrees=10, twistLimitDegrees=0}
 {bone1Name=dyn_ear04_l, bone2Name=dyn_ear03_l, swing1LimitDegrees=7, swing2LimitDegrees=10, twistLimitDegrees=0}
]
